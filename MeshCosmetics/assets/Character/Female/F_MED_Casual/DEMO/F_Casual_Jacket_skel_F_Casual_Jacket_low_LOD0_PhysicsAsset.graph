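# Physics asset for the low-LOD0 skeletal mesh of a female casual jacket cosmetic, in a demo folder.
# It holds a single capsule body, named after the mesh, placed at roughly torso height and rotated about 90 degrees.
# It defines no constraints, so there is no jointed ragdoll chain, only one coarse collision volume for the whole jacket.
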PhysicsAsset /MeshCosmetics/assets/Character/Female/F_MED_Casual/DEMO/F_Casual_Jacket_skel_F_Casual_Jacket_low_LOD0_PhysicsAsset
bodies {F_Casual_Jacket_low_LOD0 [{shapeType=Capsule, length=34.6848, radius=18.0102, center=<-0.0545 -3.0994 131.9443>, rotation=<rot 0 89.2202 -89.425>}]}
constraints []
